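# Control Rig for the ArcticIceBlue backpack that adjusts the left, center and right cape master bones during skydiving.
# A spherical pose reader on the pelvis produces a dive-mode alpha, which weights rotation offsets on all three cape masters.
# The absolute difference between the is_skydive_dive variable and that alpha is interpolated to weight separate offsets on the two side masters.
# A four-step sequence runs the reader, difference and both rotation passes in order.
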ControlRigBlueprint /Game/Accessories/FORT_Backpacks/Backpack_M_MED_ArcticIceBlue/Meshes/M_MED_ArcticIceBlue_Pack_skydive_CtrlRig
graphs [
 ContainedGraph {nodes [
   n1 = RigVMFunctionEntryNode()
   SequenceAggregate = &RigVMFunction_Sequence(ExecuteContext=@n1.ExecuteContext)
   SequenceAggregate_1 = &RigVMFunction_Sequence(ExecuteContext=@SequenceAggregate.B)
   SequenceAggregate_2 = &RigVMFunction_Sequence(ExecuteContext=@SequenceAggregate_1.B)
   RigVMFunctionReturnNode(A=@SequenceAggregate.A, B=@SequenceAggregate_1.A, C=@SequenceAggregate_2.A, D=@SequenceAggregate_2.B)
  ]}
 RigVMFunctionLibrary {links [], nodes []}
 RigVMModel {nodes [
   RigUnit_BeginExecution = &RigUnit_BeginExecution()
   SequenceAggregate = RigVMAggregateNode(ExecuteContext=@RigUnit_BeginExecution.ExecutePin)
   Subtract = &RigVMFunction_MathDoubleSub(A=RigVMVariableNode() {variable=is_skydive_dive}.Value, B=RigVMVariableNode() {variable=spr_skydive_dive_mode_alpha}.Value)
   Absolute = &RigVMFunction_MathDoubleAbs(Value=@Subtract.Result)
   RigVMVariableNode(Value=@Absolute.Result, ExecuteContext=@SequenceAggregate.B) {variable=is_skydive_subtract}
   &RigUnit_ModifyTransforms(
    ItemToModify=((Item=(Type=Bone,Name="dyn_cape_l_MSTR"),Transform=(Rotation=(X=0,Y=0,Z=0,W=1),Translation=(X=0,Y=0,Z=0),Scale3D=(X=1,Y=1,Z=1))))
    ItemToModify.0.Transform.Rotation=&RigVMFunction_MathQuaternionFromEuler(Euler=(X=10,Y=5,Z=0)).Result
    Weight=RigVMVariableNode() {variable=spr_skydive_dive_mode_alpha}.Value
    ExecutePin=@SequenceAggregate.D
   )
   > &RigUnit_ModifyTransforms(
    ItemToModify=((Item=(Type=Bone,Name="dyn_cape_MSTR"),Transform=(Rotation=(X=0,Y=0,Z=0,W=1),Translation=(X=0,Y=0,Z=0),Scale3D=(X=1,Y=1,Z=1))))
    ItemToModify.0.Transform.Rotation=&RigVMFunction_MathQuaternionFromEuler(Euler=(X=0,Y=5,Z=0)).Result
    Weight=RigVMVariableNode() {variable=spr_skydive_dive_mode_alpha}.Value
   )
   > &RigUnit_ModifyTransforms(
    ItemToModify=((Item=(Type=Bone,Name="dyn_cape_r_MSTR"),Transform=(Rotation=(X=0,Y=0,Z=0,W=1),Translation=(X=0,Y=0,Z=0),Scale3D=(X=1,Y=1,Z=1))))
    ItemToModify.0.Transform.Rotation=&RigVMFunction_MathQuaternionFromEuler(Euler=(X=-10,Y=5,Z=0)).Result
    Weight=RigVMVariableNode() {variable=spr_skydive_dive_mode_alpha}.Value
   )
   n1 = &RigUnit_SphericalPoseReader(
    ActiveRegionScaleFactors=(PositiveWidth=1,NegativeWidth=1,PositiveHeight=1,NegativeHeight=1)
    ActiveRegionSize=0.2
    Debug=(bDrawDebug=false,bDraw2D=false,DebugScale=25,DebugSegments=20,DebugThickness=0.25)
    DriverAxis=(X=1,Y=0,Z=0)
    DriverItem=(Type=Bone,Name="pelvis")
    FalloffRegionScaleFactors=(PositiveWidth=1,NegativeWidth=1,PositiveHeight=1,NegativeHeight=1)
    FalloffSize=0.4
    OptionalParentItem=(Type=Bone)
    RotationOffset=(X=90,Y=0,Z=110)
    ExecutePin=@SequenceAggregate.A
   )
   > RigVMVariableNode(Value=@n1.OutputParam) {variable=spr_skydive_dive_mode_alpha}
   AlphaInterp = &RigVMFunction_AlphaInterp(
    InterpSpeedDecreasing=10
    InterpSpeedIncreasing=10
    Scale=1
    bClampResult=False
    Value=RigVMVariableNode() {variable=is_skydive_subtract}.Value
   )
   &RigUnit_ModifyTransforms(
    ItemToModify=((Item=(Type=Bone,Name="dyn_cape_l_MSTR"),Transform=(Rotation=(X=0,Y=0,Z=0,W=1),Translation=(X=0,Y=0,Z=0),Scale3D=(X=1,Y=1,Z=1))))
    ItemToModify.0.Transform.Rotation=&RigVMFunction_MathQuaternionFromEuler(Euler=(X=-10,Y=-35,Z=0)).Result
    Weight=@AlphaInterp.Result
    ExecutePin=@SequenceAggregate.C
   )
   > &RigUnit_ModifyTransforms(
    ItemToModify=((Item=(Type=Bone,Name="dyn_cape_r_MSTR"),Transform=(Rotation=(X=0,Y=0,Z=0,W=1),Translation=(X=0,Y=0,Z=0),Scale3D=(X=1,Y=1,Z=1))))
    ItemToModify.0.Transform.Rotation=&RigVMFunction_MathQuaternionFromEuler(Euler=(X=10,Y=-35,Z=0)).Result
    Weight=@AlphaInterp.Result
   )
  ]}
]
hierarchy [
 root {loc=<0 0 0>, rot=<0 0 0 1>}
 pelvis {parent=root, loc=<0 -0.7071 93.1191>, rot=<0.0127 0.707 -0.0127 -0.707>}
 spine_01 {parent=pelvis, loc=<6.7494 -0.6388 0>, rot=<0 0 0.034 -0.9994>}
 spine_02 {parent=spine_01, loc=<7.2518 -0.035 0>, rot=<0 0 0.0238 -0.9997>}
 spine_03 {parent=spine_02, loc=<6.0104 0 0>, rot=<0 0 0.0149 -0.9999>}
 spine_04 {parent=spine_03, loc=<8.1024 0 0>, rot=<0 0 -0.0132 -0.9999>}
 spine_05 {parent=spine_04, loc=<10.0146 0 0>, rot=<0 0 -0.0015 -1>}
 neck_01 {parent=spine_05, loc=<11.3138 0.1529 0>, rot=<0 0 -0.2003 -0.9797>}
 neck_02 {parent=neck_01, loc=<4.5627 0 0>, rot=<0 0 0.0504 -0.9987>}
 clavicle_l {parent=spine_05, loc=<4.3575 -0.1466 -2.3223>, rot=<0.0465 -0.7289 -0.009 -0.6829>}
 upperarm_l {parent=clavicle_l, loc=<15.5111 0 0>, rot=<0.0226 -0.3234 -0.0088 -0.9459>}
 lowerarm_l {parent=upperarm_l, loc=<26.6514 0 0>, rot=<0.0391 -0.0099 -0.2453 -0.9686>}
 upperarm_twist_01_l {parent=upperarm_l, loc=<0.0041 0 0>, rot=<0 0 0 -1>}
 upperarm_twist_02_l {parent=upperarm_l, loc=<13.7941 0 0>, rot=<0 0 0 -1>}
 dyn_L_shoulderPad {parent=clavicle_l, loc=<14.158 -1.6687 4.6533>, rot=<-0.0015 0.1258 -0.0252 -0.9917>}
 dyn_L_frontPull {parent=clavicle_l, loc=<18.2383 10.205 0.7455>, rot=<-0.2259 -0.0968 -0.0215 -0.9691>}
 dyn_L_capeSleeve_follow {parent=clavicle_l, loc=<21.806 0.518 7.9401>, rot=<-0.0265 -0.4063 -0.0788 -0.9099>}
 clavicle_r {parent=spine_05, loc=<4.3575 -0.1466 2.3223>, rot=<0.7289 0.0465 -0.6829 0.009>}
 upperarm_r {parent=clavicle_r, loc=<-15.5111 0 0>, rot=<0.0226 -0.3234 -0.0088 -0.9459>}
 lowerarm_r {parent=upperarm_r, loc=<-26.6514 0 0>, rot=<0.0391 -0.0099 -0.2453 -0.9686>}
 upperarm_twist_01_r {parent=upperarm_r, loc=<-0.0041 0 0>, rot=<0 0 0 -1>}
 upperarm_twist_02_r {parent=upperarm_r, loc=<-13.7941 0 0>, rot=<0 0 0 -1>}
 dyn_R_shoulderPad {parent=clavicle_r, loc=<-14.158 1.6687 -4.6536>, rot=<-0.0015 0.1258 -0.0252 -0.9917>}
 dyn_R_frontPull {parent=clavicle_r, loc=<-18.2383 -10.205 -0.7455>, rot=<-0.2259 -0.0968 -0.0215 -0.9691>}
 dyn_R_capeSleeve_follow {parent=clavicle_r, loc=<-21.806 -0.518 -7.9401>, rot=<-0.0265 -0.4063 -0.0788 -0.9099>}
 backpack_root {parent=spine_05, loc=<5.7306 -8.769 0>, rot=<0.0284 -0.7065 -0.0284 -0.7065>}
 hider {parent=backpack_root, loc=<0 -2.5808 0.0076>, rot=<0 0 0 1>}
 cape_top {parent=hider, loc=<0 0.0288 -2.7904>, rot=<0 0 0 1>}
 dyn_cape_r_MSTR {parent=cape_top, loc=<-16.2328 1.5541 -2.5734>, rot=<-0.4358 -0.4358 0.5569 -0.5569>}
 dyn_cape_r_01 {parent=dyn_cape_r_MSTR, loc=<0.3 0 0>, rot=<0 0 0 1>}
 dyn_cape_r_02 {parent=dyn_cape_r_01, loc=<16.1032 0 0>, rot=<0 -0.0648 0 -0.9979>}
 dyn_cape_r_03 {parent=dyn_cape_r_02, loc=<16.1032 0 0>, rot=<0 -0.0149 0 -0.9999>}
 dyn_cape_r_04 {parent=dyn_cape_r_03, loc=<16.1033 0 0>, rot=<0 -0.0046 0 -1>}
 dyn_cape_r_05 {parent=dyn_cape_r_04, loc=<16.1033 0 0>, rot=<0 0 0 1>}
 dyn_cape_r_06 {parent=dyn_cape_r_05, loc=<16.1032 0 0>, rot=<0 0 0 1>}
 dyn_cape_r_07 {parent=dyn_cape_r_06, loc=<16.1031 0 0.0001>, rot=<0 0 0 1>}
 dyn_cape_l_MSTR {parent=cape_top, loc=<16.2328 1.5541 -2.5734>, rot=<-0.4358 -0.4358 0.5569 -0.5569>}
 dyn_cape_l_01 {parent=dyn_cape_l_MSTR, loc=<0.3 0 0>, rot=<0 0 0 1>}
 dyn_cape_l_02 {parent=dyn_cape_l_01, loc=<16.1032 0 0>, rot=<0 -0.0648 0 -0.9979>}
 dyn_cape_l_03 {parent=dyn_cape_l_02, loc=<16.1032 0 0>, rot=<0 -0.0149 0 -0.9999>}
 dyn_cape_l_04 {parent=dyn_cape_l_03, loc=<16.1032 0 0>, rot=<0 -0.0046 0 -1>}
 dyn_cape_l_05 {parent=dyn_cape_l_04, loc=<16.1032 0 0>, rot=<0 0 0 1>}
 dyn_cape_l_06 {parent=dyn_cape_l_05, loc=<16.1032 0 0>, rot=<0 0 0 1>}
 dyn_cape_l_07 {parent=dyn_cape_l_06, loc=<16.1032 0 0>, rot=<0 0 0 1>}
 dyn_cape_MSTR {parent=cape_top, loc=<0 -0.0317 -2.5734>, rot=<-0.4358 -0.4358 0.5569 -0.5569>}
 dyn_cape_01 {parent=dyn_cape_MSTR, loc=<0.3 0 0>, rot=<0 0 0 1>}
 dyn_cape_02 {parent=dyn_cape_01, loc=<16.1032 0 0>, rot=<0 -0.0648 0 -0.9979>}
 dyn_cape_03 {parent=dyn_cape_02, loc=<16.1032 0 0>, rot=<0 -0.0149 0 -0.9999>}
 dyn_cape_04 {parent=dyn_cape_03, loc=<16.1032 0 0>, rot=<0 -0.0046 0 -1>}
 dyn_cape_05 {parent=dyn_cape_04, loc=<16.1032 0 0>, rot=<0 0 0 1>}
 dyn_cape_06 {parent=dyn_cape_05, loc=<16.1032 0 0>, rot=<0 0 0 1>}
 dyn_cape_07 {parent=dyn_cape_06, loc=<16.1032 0 0>, rot=<0 0 0 1>}
 thigh_l {parent=pelvis, loc=<-2.5693 -0.0109 -7.7501>, rot=<0.0012 0.0572 0.0155 -0.9982>}
 calf_l {parent=thigh_l, loc=<-39.0484 0 0>, rot=<0.0006 -0.0161 -0.0391 -0.9991>}
 foot_l {parent=calf_l, loc=<-42.561 0.1077 0.4292>, rot=<0.0418 -0.0455 0.0343 -0.9975>}
 ball_l {parent=foot_l, loc=<-6.1599 13.8264 0.6806>, rot=<0.0001 -0.0001 -0.7063 -0.7079>}
 calf_twist_02_l {parent=calf_l, loc=<-18.8064 -0.0967 0.0397>, rot=<0.0237 0.0095 -0.0035 -0.9997>}
 dyn_L_pantLeg_out_B {parent=calf_twist_02_l, loc=<-0.4996 3.0731 -7.3796>, rot=<-0.0461 -0.998 -0.0226 -0.0371>}
 calf_twist_01_l {parent=calf_l, loc=<-35.813 -0.0967 0.0397>, rot=<0.0237 0.0095 -0.0035 -0.9997>}
 dyn_L_pantLeg_out_A {parent=calf_l, loc=<-2.8837 1.2609 -6.2672>, rot=<-0.0415 -0.998 0.0006 -0.0476>}
 thigh_twist_01_l {parent=thigh_l, loc=<-21.6303 0 0>, rot=<0.0013 0 0 -1>}
 thigh_r {parent=pelvis, loc=<-2.5693 -0.0109 7.7501>, rot=<0.0572 -0.0012 0.9982 0.0155>}
 calf_r {parent=thigh_r, loc=<39.0484 0 0>, rot=<0.0006 -0.0161 -0.0391 -0.9991>}
 foot_r {parent=calf_r, loc=<42.561 -0.1077 -0.4292>, rot=<0.0418 -0.0455 0.0343 -0.9975>}
 ball_r {parent=foot_r, loc=<6.1599 -13.8264 -0.6806>, rot=<0.0001 -0.0001 -0.7063 -0.7079>}
 calf_twist_02_r {parent=calf_r, loc=<18.8064 0.0967 -0.0397>, rot=<0.0237 0.0095 -0.0035 -0.9997>}
 dyn_R_pantLeg_out_B {parent=calf_twist_02_r, loc=<0.4995 -3.0731 7.3796>, rot=<0.998 -0.0461 0.0371 -0.0226>}
 calf_twist_01_r {parent=calf_r, loc=<35.813 0.0967 -0.0397>, rot=<0.0237 0.0095 -0.0035 -0.9997>}
 dyn_R_pantLeg_out_A {parent=calf_r, loc=<2.8837 -1.2609 6.2672>, rot=<0.998 -0.0415 0.0476 0.0006>}
 thigh_twist_01_r {parent=thigh_r, loc=<21.6303 0 0>, rot=<0.0013 0 0 -1>}
 MaterialCurveA {type=Curve, loc=<0 0 0>, rot=<0 0 0 1>}
]
variables [
 is_skydive_dive {type=double, default=0}
 is_skydive_subtract {type=double, default=0}
 spr_skydive_dive_mode_alpha {type=double, default=0}
]
