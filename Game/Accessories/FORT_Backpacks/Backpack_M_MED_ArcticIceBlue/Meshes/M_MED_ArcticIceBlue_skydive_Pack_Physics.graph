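# Physics asset for the skydive pack mesh of the ArcticIceBlue backpack, simulating a cape.
# Left, center and right capsule chains each run seven segments from a master bone with limited swing and twist; only the side masters constrain to spine_05.
# Side segments 02–07 lock angularly to their center counterparts, tying the strips into one sheet.
# Spine, pelvis and limb capsules, linked by free-rotation joints, plus a root box provide body shapes.
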PhysicsAsset /Game/Accessories/FORT_Backpacks/Backpack_M_MED_ArcticIceBlue/Meshes/M_MED_ArcticIceBlue_skydive_Pack_Physics
bodies {
 calf_l [{shapeType=Capsule, length=24.5432, radius=15.5, shapeName=calf_l_capsule, center=<-16.2682 0 0>, rotation=<rot 90 0 0>}]
 calf_r [{shapeType=Capsule, length=24.5432, radius=15.5, shapeName=calf_r_capsule, center=<16.2682 0 0>, rotation=<rot 90 0 0>}]
 dyn_cape_01 [{shapeType=Capsule, length=20, radius=9, shapeName=dyn_cape_01_capsule, center=<6 0 3>, rotation=<rot 0 0 90>}]
 dyn_cape_02 [{shapeType=Capsule, length=20, radius=9.5, shapeName=dyn_cape_01_capsule, center=<8 0 3>, rotation=<rot 0 0 90>}]
 dyn_cape_03 [{shapeType=Capsule, length=20, radius=8.5, shapeName=dyn_cape_01_capsule, center=<10 0 2>, rotation=<rot 0 0 90>}]
 dyn_cape_04 [{shapeType=Capsule, length=20, radius=8.5, shapeName=dyn_cape_01_capsule, center=<10 0 3>, rotation=<rot 0 0 90>}]
 dyn_cape_05 [{shapeType=Capsule, length=20, radius=8.5, shapeName=dyn_cape_01_capsule, center=<12 0 3>, rotation=<rot 0 0 90>}]
 dyn_cape_06 [{shapeType=Capsule, length=32.9035, radius=11, shapeName=dyn_cape_01_capsule, center=<13 0 2>, rotation=<rot 0 0 90>}]
 dyn_cape_07 [{shapeType=Capsule, length=32.9035, radius=11, shapeName=dyn_cape_01_capsule, center=<12 0 2>, rotation=<rot 0 0 90>}]
 dyn_cape_MSTR [{shapeType=Capsule, length=0.5, radius=0.5, shapeName=dyn_cape_01_capsule}]
 dyn_cape_l_01 [{shapeType=Capsule, length=20, radius=11, shapeName=dyn_cape_01_capsule, center=<6.6765 0 8.5>, rotation=<rot 0 0 90>}]
 dyn_cape_l_02 [{shapeType=Capsule, length=25, radius=11.5, shapeName=dyn_cape_01_capsule, center=<8 0 3>, rotation=<rot 0 0 90>}]
 dyn_cape_l_03 [{shapeType=Capsule, length=20, radius=10, shapeName=dyn_cape_01_capsule, center=<10 0 2>, rotation=<rot 0 0 90>}]
 dyn_cape_l_04 [{shapeType=Capsule, length=20, radius=9, shapeName=dyn_cape_01_capsule, center=<10 0 3>, rotation=<rot 0 0 90>}]
 dyn_cape_l_05 [{shapeType=Capsule, length=20, radius=8.5, shapeName=dyn_cape_01_capsule, center=<12 0 3>, rotation=<rot 0 0 90>}]
 dyn_cape_l_06 [{shapeType=Capsule, length=20, radius=9, shapeName=dyn_cape_01_capsule, center=<13 0 2>, rotation=<rot 0 0 90>}]
 dyn_cape_l_07 [{shapeType=Capsule, length=20, radius=9, shapeName=dyn_cape_01_capsule, center=<12 0 2>, rotation=<rot 0 0 90>}]
 dyn_cape_l_MSTR [{shapeType=Capsule, length=0.5, radius=0.5, shapeName=dyn_cape_01_capsule}]
 dyn_cape_r_01 [{shapeType=Capsule, length=20, radius=11, shapeName=dyn_cape_01_capsule, center=<6.6765 0 8.5>, rotation=<rot 0 0 90>}]
 dyn_cape_r_02 [{shapeType=Capsule, length=25, radius=11.5, shapeName=dyn_cape_01_capsule, center=<8 0 3>, rotation=<rot 0 0 90>}]
 dyn_cape_r_03 [{shapeType=Capsule, length=20, radius=10, shapeName=dyn_cape_01_capsule, center=<10 0 2>, rotation=<rot 0 0 90>}]
 dyn_cape_r_04 [{shapeType=Capsule, length=20, radius=9, shapeName=dyn_cape_01_capsule, center=<10 0 3>, rotation=<rot 0 0 90>}]
 dyn_cape_r_05 [{shapeType=Capsule, length=20, radius=8.5, shapeName=dyn_cape_01_capsule, center=<12 0 3>, rotation=<rot 0 0 90>}]
 dyn_cape_r_06 [{shapeType=Capsule, length=20, radius=9, shapeName=dyn_cape_01_capsule, center=<13 0 2>, rotation=<rot 0 0 90>}]
 dyn_cape_r_07 [{shapeType=Capsule, length=20, radius=9, shapeName=dyn_cape_01_capsule, center=<12 0 2>, rotation=<rot 0 0 90>}]
 dyn_cape_r_MSTR [{shapeType=Capsule, length=0.5, radius=0.5, shapeName=dyn_cape_01_capsule}]
 foot_l [{shapeType=Capsule, length=24.7259, radius=13, shapeName=foot_l_capsule, center=<-4.0407 6.4709 0.5648>, rotation=<rot 0 0 90>}]
 foot_r [{shapeType=Capsule, length=24.7259, radius=13, shapeName=foot_r_capsule, center=<4.0407 -6.4709 -0.5648>, rotation=<rot 0 0 90>}]
 lowerarm_l [{shapeType=Capsule, length=12, radius=12, shapeName=upperarm_l_capsule, center=<11.1469 2.3136 0.2476>, rotation=<rot 90 -121.2523 -126.2523>}]
 lowerarm_r [{
   shapeType=Capsule
   length=12
   radius=12
   shapeName=upperarm_r_capsule
   center=<-11.1469 -2.3136 -0.2476>
   rotation=<rot 90 -121.2523 -126.2523>
  }]
 pelvis [{shapeType=Capsule, length=14, radius=16, shapeName=spine_01_capsule, center=<-2.1069 3.3473 0>}]
 root [{shapeType=Box, extentX=250, extentY=250, extentZ=30, shapeName=root_box, center=<0 0 -14.5752>, rotation=<rot 0 -90 0>}]
 spine_01 [{shapeType=Capsule, length=14, radius=14, shapeName=spine_01_capsule, center=<-0.0285 3.2728 0>}]
 spine_02 [{shapeType=Capsule, length=14, radius=14, shapeName=spine_02_capsule, center=<-0.0723 3.6143 0>}]
 spine_04 [{shapeType=Capsule, length=14, radius=13, shapeName=spine_04_capsule, center=<2.298 0.4883 0>}]
 spine_05 [{shapeType=Capsule, length=14, radius=13, shapeName=spine_05_capsule, center=<-0.2973 3.6987 0>}]
 thigh_l [{shapeType=Capsule, length=10, radius=15, shapeName=thigh_l_capsule, center=<-18.5872 4.3916 1.0625>, rotation=<rot 90 0 0>}]
 thigh_r [{shapeType=Capsule, length=10, radius=15, shapeName=thigh_r_capsule, center=<18.5873 -4.3915 -1.0625>, rotation=<rot 90 0 0>}]
 upperarm_l [{shapeType=Capsule, length=19.9161, radius=10, shapeName=upperarm_l_capsule, center=<10.0017 -0.0489 -0.0007>, rotation=<rot 90 180 174.9999>}]
 upperarm_r [{shapeType=Capsule, length=19.9161, radius=10, shapeName=upperarm_r_capsule, center=<-10.0017 0.0489 0.0007>, rotation=<rot 90 180 174.9999>}]
}
constraints [
 {bone1Name=dyn_cape_l_01, bone2Name=dyn_cape_l_MSTR, swing1LimitDegrees=10, swing2LimitDegrees=25, twistLimitDegrees=10}
 {bone1Name=dyn_cape_l_02, bone2Name=dyn_cape_l_01, swing1LimitDegrees=10, swing2LimitDegrees=10, twistLimitDegrees=10}
 {bone1Name=dyn_cape_l_03, bone2Name=dyn_cape_l_02, swing1LimitDegrees=30, swing2LimitDegrees=10, twistLimitDegrees=10}
 {bone1Name=dyn_cape_l_04, bone2Name=dyn_cape_l_03, swing1LimitDegrees=30, swing2LimitDegrees=30, twistLimitDegrees=10}
 {bone1Name=dyn_cape_l_05, bone2Name=dyn_cape_l_04, swing1LimitDegrees=30, swing2LimitDegrees=20, twistLimitDegrees=10}
 {bone1Name=dyn_cape_l_06, bone2Name=dyn_cape_l_05, swing1LimitDegrees=30, swing2LimitDegrees=20, twistLimitDegrees=10}
 {bone1Name=dyn_cape_l_07, bone2Name=dyn_cape_l_06, swing1LimitDegrees=30, swing2LimitDegrees=20, twistLimitDegrees=10}
 {bone1Name=dyn_cape_r_01, bone2Name=dyn_cape_r_MSTR, swing1LimitDegrees=10, swing2LimitDegrees=25, twistLimitDegrees=10}
 {bone1Name=dyn_cape_r_02, bone2Name=dyn_cape_r_01, swing1LimitDegrees=10, swing2LimitDegrees=10, twistLimitDegrees=10}
 {bone1Name=dyn_cape_r_03, bone2Name=dyn_cape_r_02, swing1LimitDegrees=30, swing2LimitDegrees=10, twistLimitDegrees=10}
 {bone1Name=dyn_cape_r_04, bone2Name=dyn_cape_r_03, swing1LimitDegrees=30, swing2LimitDegrees=30, twistLimitDegrees=10}
 {bone1Name=dyn_cape_r_05, bone2Name=dyn_cape_r_04, swing1LimitDegrees=30, swing2LimitDegrees=20, twistLimitDegrees=10}
 {bone1Name=dyn_cape_r_06, bone2Name=dyn_cape_r_05, swing1LimitDegrees=30, swing2LimitDegrees=20, twistLimitDegrees=10}
 {bone1Name=dyn_cape_r_07, bone2Name=dyn_cape_r_06, swing1LimitDegrees=30, swing2LimitDegrees=20, twistLimitDegrees=10}
 {bone1Name=dyn_cape_l_MSTR, bone2Name=spine_05, swing1LimitDegrees=45, swing2LimitDegrees=15, twistLimitDegrees=1}
 {bone1Name=dyn_cape_r_MSTR, bone2Name=spine_05, swing1LimitDegrees=45, swing2LimitDegrees=15, twistLimitDegrees=1}
 {
  bone1Name=upperarm_l
  bone2Name=spine_05
  swing1LimitDegrees=45
  swing1Motion=Free
  swing2LimitDegrees=45
  swing2Motion=Free
  twistLimitDegrees=45
  twistMotion=Free
 }
 {
  bone1Name=upperarm_r
  bone2Name=spine_05
  swing1LimitDegrees=45
  swing1Motion=Free
  swing2LimitDegrees=45
  swing2Motion=Free
  twistLimitDegrees=45
  twistMotion=Free
 }
 {
  bone1Name=spine_02
  bone2Name=spine_01
  swing1LimitDegrees=45
  swing1Motion=Free
  swing2LimitDegrees=45
  swing2Motion=Free
  twistLimitDegrees=45
  twistMotion=Free
 }
 {
  bone1Name=spine_04
  bone2Name=spine_02
  swing1LimitDegrees=45
  swing1Motion=Free
  swing2LimitDegrees=45
  swing2Motion=Free
  twistLimitDegrees=45
  twistMotion=Free
 }
 {
  bone1Name=spine_05
  bone2Name=spine_04
  swing1LimitDegrees=45
  swing1Motion=Free
  swing2LimitDegrees=45
  swing2Motion=Free
  twistLimitDegrees=45
  twistMotion=Free
 }
 {
  bone1Name=calf_l
  bone2Name=thigh_l
  swing1LimitDegrees=45
  swing1Motion=Free
  swing2LimitDegrees=45
  swing2Motion=Free
  twistLimitDegrees=45
  twistMotion=Free
 }
 {
  bone1Name=calf_r
  bone2Name=thigh_r
  swing1LimitDegrees=45
  swing1Motion=Free
  swing2LimitDegrees=45
  swing2Motion=Free
  twistLimitDegrees=45
  twistMotion=Free
 }
 {
  bone1Name=foot_r
  bone2Name=calf_r
  swing1LimitDegrees=45
  swing1Motion=Free
  swing2LimitDegrees=45
  swing2Motion=Free
  twistLimitDegrees=45
  twistMotion=Free
 }
 {
  bone1Name=foot_l
  bone2Name=calf_l
  swing1LimitDegrees=45
  swing1Motion=Free
  swing2LimitDegrees=45
  swing2Motion=Free
  twistLimitDegrees=45
  twistMotion=Free
 }
 {
  bone1Name=lowerarm_l
  bone2Name=upperarm_l
  swing1LimitDegrees=45
  swing1Motion=Free
  swing2LimitDegrees=45
  swing2Motion=Free
  twistLimitDegrees=45
  twistMotion=Free
 }
 {
  bone1Name=lowerarm_r
  bone2Name=upperarm_r
  swing1LimitDegrees=45
  swing1Motion=Free
  swing2LimitDegrees=45
  swing2Motion=Free
  twistLimitDegrees=45
  twistMotion=Free
 }
 {
  bone1Name=spine_01
  bone2Name=pelvis
  swing1LimitDegrees=45
  swing1Motion=Free
  swing2LimitDegrees=45
  swing2Motion=Free
  twistLimitDegrees=45
  twistMotion=Free
 }
 {
  bone1Name=thigh_l
  bone2Name=pelvis
  swing1LimitDegrees=45
  swing1Motion=Free
  swing2LimitDegrees=45
  swing2Motion=Free
  twistLimitDegrees=45
  twistMotion=Free
 }
 {
  bone1Name=thigh_r
  bone2Name=pelvis
  swing1LimitDegrees=45
  swing1Motion=Free
  swing2LimitDegrees=45
  swing2Motion=Free
  twistLimitDegrees=45
  twistMotion=Free
 }
 {bone1Name=dyn_cape_06, bone2Name=dyn_cape_05, swing1LimitDegrees=20, swing2LimitDegrees=20, twistLimitDegrees=10}
 {bone1Name=dyn_cape_07, bone2Name=dyn_cape_06, swing1LimitDegrees=20, swing2LimitDegrees=20, twistLimitDegrees=10}
 {
  bone1Name=dyn_cape_r_07
  bone2Name=dyn_cape_07
  swing1LimitDegrees=45
  swing1Motion=Locked
  swing2LimitDegrees=45
  swing2Motion=Locked
  twistLimitDegrees=45
  twistMotion=Locked
 }
 {
  bone1Name=dyn_cape_l_07
  bone2Name=dyn_cape_07
  swing1LimitDegrees=45
  swing1Motion=Locked
  swing2LimitDegrees=45
  swing2Motion=Locked
  twistLimitDegrees=45
  twistMotion=Locked
 }
 {
  bone1Name=dyn_cape_r_06
  bone2Name=dyn_cape_06
  swing1LimitDegrees=45
  swing1Motion=Locked
  swing2LimitDegrees=45
  swing2Motion=Locked
  twistLimitDegrees=45
  twistMotion=Locked
 }
 {
  bone1Name=dyn_cape_l_06
  bone2Name=dyn_cape_06
  swing1LimitDegrees=45
  swing1Motion=Locked
  swing2LimitDegrees=45
  swing2Motion=Locked
  twistLimitDegrees=45
  twistMotion=Locked
 }
 {bone1Name=dyn_cape_05, bone2Name=dyn_cape_04, swing1LimitDegrees=45, swing2LimitDegrees=45, twistLimitDegrees=45}
 {bone1Name=dyn_cape_04, bone2Name=dyn_cape_03, swing1LimitDegrees=45, swing2LimitDegrees=45, twistLimitDegrees=45}
 {bone1Name=dyn_cape_03, bone2Name=dyn_cape_02, swing1LimitDegrees=45, swing2LimitDegrees=45, twistLimitDegrees=45}
 {bone1Name=dyn_cape_02, bone2Name=dyn_cape_01, swing1LimitDegrees=45, swing2LimitDegrees=10, twistLimitDegrees=45}
 {bone1Name=dyn_cape_01, bone2Name=dyn_cape_MSTR, swing1LimitDegrees=10, swing2LimitDegrees=15, twistLimitDegrees=10}
 {
  bone1Name=dyn_cape_r_05
  bone2Name=dyn_cape_05
  swing1LimitDegrees=45
  swing1Motion=Locked
  swing2LimitDegrees=45
  swing2Motion=Locked
  twistLimitDegrees=45
  twistMotion=Locked
 }
 {
  bone1Name=dyn_cape_l_05
  bone2Name=dyn_cape_05
  swing1LimitDegrees=45
  swing1Motion=Locked
  swing2LimitDegrees=45
  swing2Motion=Locked
  twistLimitDegrees=45
  twistMotion=Locked
 }
 {
  bone1Name=dyn_cape_r_04
  bone2Name=dyn_cape_04
  swing1LimitDegrees=45
  swing1Motion=Locked
  swing2LimitDegrees=45
  swing2Motion=Locked
  twistLimitDegrees=45
  twistMotion=Locked
 }
 {
  bone1Name=dyn_cape_l_04
  bone2Name=dyn_cape_04
  swing1LimitDegrees=45
  swing1Motion=Locked
  swing2LimitDegrees=45
  swing2Motion=Locked
  twistLimitDegrees=45
  twistMotion=Locked
 }
 {
  bone1Name=dyn_cape_r_03
  bone2Name=dyn_cape_03
  swing1LimitDegrees=45
  swing1Motion=Locked
  swing2LimitDegrees=45
  swing2Motion=Locked
  twistLimitDegrees=45
  twistMotion=Locked
 }
 {
  bone1Name=dyn_cape_l_03
  bone2Name=dyn_cape_03
  swing1LimitDegrees=45
  swing1Motion=Locked
  swing2LimitDegrees=45
  swing2Motion=Locked
  twistLimitDegrees=45
  twistMotion=Locked
 }
 {
  bone1Name=dyn_cape_r_02
  bone2Name=dyn_cape_02
  swing1LimitDegrees=45
  swing1Motion=Locked
  swing2LimitDegrees=45
  swing2Motion=Locked
  twistLimitDegrees=45
  twistMotion=Locked
 }
 {
  bone1Name=dyn_cape_l_02
  bone2Name=dyn_cape_02
  swing1LimitDegrees=45
  swing1Motion=Locked
  swing2LimitDegrees=45
  swing2Motion=Locked
  twistLimitDegrees=45
  twistMotion=Locked
 }
 {
  bone1Name=pelvis
  bone2Name=root
  swing1LimitDegrees=45
  swing1Motion=Free
  swing2LimitDegrees=45
  swing2Motion=Free
  twistLimitDegrees=45
  twistMotion=Free
 }
]
