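# Animation Blueprint for the ArcticIceBlue backpack's cape, following the attached character's pose and reacting to movement.
# Copied pose passes through a Control Rig and arm-driven constraints, then feeds either trail controllers (low LOD or rigid body disabled) or RigidBody simulation.
# The simulated branch swaps physics assets for skydiving, then adds a skydive rig, dive constraints and sprint offsets.
# Update logic derives movement and arm alphas, LOD choice and state-based gravity.
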AnimBlueprint /Game/Accessories/FORT_Backpacks/Backpack_M_MED_ArcticIceBlue/Meshes/M_MED_ArcticIceBlue_Pack_AnimBP
target_skeleton=&/M_MED_ArcticIceBlue_Pack_Skeleton
graphs [
 AnimGraph {nodes [
   Output Pose(Result=Use cached pose 'trailCache'()) {layerGroup=None}
   n1 = Control Rig!(Source=Copy Pose From Mesh() {bCopyCurves=true, bUseAttachedParent=true}, Alpha, parachuteOpened:bool=Get bIsParachuteOpened()) -> (Pose) {
    alphaBoolBlend {blendOption=Linear}
    controlRigAssetReference {blueprintRigClass=&/M_MED_ArcticIceBlue_Pack_CtrlRig.M_MED_ArcticIceBlue_Pack_CtrlRig_C}
   }
   n2 = "Constraint - dyn_L_frontPull (P)"(ComponentPose=Local To Component(LocalPose=@n1), Alpha=Get L armForward() -> (L armForward:double), ConstraintWeights_0=1) {
    alphaBoolBlend {blendOption=Linear}
    boneToModify {boneName=dyn_L_frontPull}
    constraintSetup [{targetBone {boneName=upperarm_twist_01_l}}]
    constraintWeights [1]
   }
   n3 = "Constraint - dyn_R_frontPull (P)"(ComponentPose=@n2, Alpha=Get R armForward() -> (R armForward:double), ConstraintWeights_0=1) {
    alphaBoolBlend {blendOption=Linear}
    boneToModify {boneName=dyn_R_frontPull}
    constraintSetup [{targetBone {boneName=upperarm_twist_01_r}}]
    constraintWeights [1]
   }
   n4 = "Constraint - dyn_L_capeSleeve_follow (P)"(ComponentPose=@n3, Alpha=Get L armUp() -> (L armUp:double), ConstraintWeights_0=1) {
    alphaBoolBlend {blendOption=Linear}
    boneToModify {boneName=dyn_L_capeSleeve_follow}
    constraintSetup [{targetBone {boneName=upperarm_twist_01_l}}]
    constraintWeights [1]
   }
   n5 = "Constraint - dyn_R_capeSleeve_follow (P)"(ComponentPose=@n4, Alpha=Get R armUp() -> (R armUp:double), ConstraintWeights_0=1) {
    alphaBoolBlend {blendOption=Linear}
    boneToModify {boneName=dyn_R_capeSleeve_follow}
    constraintSetup [{targetBone {boneName=upperarm_twist_01_r}}]
    constraintWeights [1]
   }
   Save cached pose 'Start_Cache'(Pose=Component To Local(ComponentPose=@n5)) {cacheName=Start_Cache}
   n6 = "Transform (Modify) Bone - Bone: backpack_root"!(
    ComponentPose=Local To Component(LocalPose=Use cached pose 'Start_Cache'())
    bAlphaBoolEnabled:bool=NOT Boolean(A=Get bIsInFrontEnd())
    Translation="0,0,0"
    Rotation=Get SkydivingAlpha() -> (SkydivingAlpha:Rotator)
    Scale="1,1,1"
   ) -> (Pose) {
    alphaInputType=Bool
    rotationMode=BMM_Additive
    rotationSpace=BCS_BoneSpace
    alphaBoolBlend {blendOption=Linear}
    boneToModify {boneName=backpack_root}
   }
   n7 = Get bIsSkyDiving()
   n8 = "Transform (Modify) Bone - Bone: dyn_cape_l_04"!(ComponentPose=@n6, bAlphaBoolEnabled:bool=@n7, Translation="0,0,0", Rotation="20,0,0", Scale="1,1,1") -> (Pose) {
    alphaInputType=Bool
    rotationMode=BMM_Additive
    rotationSpace=BCS_BoneSpace
    alphaBoolBlend {blendOption=Linear}
    boneToModify {boneName=dyn_cape_l_04}
    rotation {pitch=20}
   }
   n9 = "Transform (Modify) Bone - Bone: dyn_cape_r_04"!(ComponentPose=@n8, bAlphaBoolEnabled:bool=@n7, Translation="0,0,0", Rotation="20,0,0", Scale="1,1,1") -> (Pose) {
    alphaInputType=Bool
    rotationMode=BMM_Additive
    rotationSpace=BCS_BoneSpace
    alphaBoolBlend {blendOption=Linear}
    boneToModify {boneName=dyn_cape_r_04}
    rotation {pitch=20}
   }
   n10 = "Transform (Modify) Bone - Bone: dyn_cape_l_06"!(ComponentPose=@n9, bAlphaBoolEnabled:bool=@n7, Translation="0,0,0", Rotation="10,0,0", Scale="1,1,1") -> (Pose) {
    alphaInputType=Bool
    rotationMode=BMM_Additive
    rotationSpace=BCS_BoneSpace
    alphaBoolBlend {blendOption=Linear}
    boneToModify {boneName=dyn_cape_l_06}
    rotation {pitch=10}
   }
   n11 = "Transform (Modify) Bone - Bone: dyn_cape_r_06"!(ComponentPose=@n10, bAlphaBoolEnabled:bool=@n7, Translation="0,0,0", Rotation="10,0,0", Scale="1,1,1") -> (Pose) {
    alphaInputType=Bool
    rotationMode=BMM_Additive
    rotationSpace=BCS_BoneSpace
    alphaBoolBlend {blendOption=Linear}
    boneToModify {boneName=dyn_cape_r_06}
    rotation {pitch=10}
   }
   n12 = Get BwdMovement() -> (BwdMovement:double)
   n13 = Trail controller - Bone: dyn_cape_07(ComponentPose=@n11, Alpha=@n12) {
    bLimitRotation=true
    chainLength=7
    debugLifeTime=0.1
    maxDeltaTime=0.333
    alphaScaleBiasClamp {bInterpResult=true}
    planarLimits [{drivingBone {boneName=backpack_root}, planeTransform {location=<0 -3 0>, rotation=<rot 0 0 -90>, scale=<1 1 1>}}]
    rotationLimits [
     {limitMax=<180 180 0>, limitMin=<-180 -180 -70>}
     {limitMax=<180 180 0>, limitMin=<-180 -180 -70>}
     {limitMax=<180 180 0>, limitMin=<-180 -180 -70>}
     {limitMax=<180 180 0>, limitMin=<-180 -180 -70>}
     {limitMax=<180 180 0>, limitMin=<-180 -180 -70>}
     {limitMax=<180 180 0>, limitMin=<-180 -180 -70>}
     {limitMax=<180 180 0>, limitMin=<-180 -180 -70>}
    ]
    rotationOffsets [<0 0 0>, <0 0 0>, <0 0 0>, <0 0 0>, <0 0 0>, <0 0 0>, <0 0 0>]
    trailBone {boneName=dyn_cape_07}
    trailRelaxationSpeed {editorCurveData {keys [<at -0.0134 19.3841 RCIM_Linear>,<at 0.5079 12.8986 RCIM_Linear>,<at 0.9955 9.6044 RCIM_Linear>]}}
   }
   n14 = Trail controller - Bone: dyn_cape_l_07(ComponentPose=@n13, Alpha=@n12) {
    bLimitRotation=true
    chainLength=7
    debugLifeTime=0.1
    maxDeltaTime=0.0333
    alphaScaleBiasClamp {bInterpResult=true}
    planarLimits [{drivingBone {boneName=backpack_root}, planeTransform {location=<0 0 0>, rotation=<rot 0 0 -90>, scale=<1 1 1>}}]
    rotationLimits [
     {limitMax=<180 180 0>, limitMin=<-180 -180 -70>}
     {limitMax=<180 180 0>, limitMin=<-180 -180 -70>}
     {limitMax=<180 180 0>, limitMin=<-180 -180 -70>}
     {limitMax=<180 180 0>, limitMin=<-180 -180 -70>}
     {limitMax=<180 180 0>, limitMin=<-180 -180 -70>}
     {limitMax=<180 180 0>, limitMin=<-180 -180 -70>}
     {limitMax=<180 180 0>, limitMin=<-180 -180 -70>}
    ]
    rotationOffsets [<0 0 0>, <0 0 0>, <0 0 0>, <0 0 0>, <0 0 0>, <0 0 0>, <0 0 0>]
    trailBone {boneName=dyn_cape_l_07}
    trailRelaxationSpeed {editorCurveData {keys [<at 0 15 RCIM_Linear>,<at 0.5289 13.1838 RCIM_Linear>,<at 1 6 RCIM_Linear>]}}
   }
   n15 = Trail controller - Bone: dyn_cape_r_07(ComponentPose=@n14, Alpha=@n12) {
    bLimitRotation=true
    chainLength=7
    debugLifeTime=0.1
    maxDeltaTime=0.0333
    alphaScaleBiasClamp {bInterpResult=true}
    planarLimits [{drivingBone {boneName=backpack_root}, planeTransform {location=<0 0 0>, rotation=<rot 0 0 -100>, scale=<1 1 1>}}]
    rotationLimits [
     {limitMax=<180 180 0>, limitMin=<-180 -180 -70>}
     {limitMax=<180 180 0>, limitMin=<-180 -180 -70>}
     {limitMax=<180 180 0>, limitMin=<-180 -180 -70>}
     {limitMax=<180 180 0>, limitMin=<-180 -180 -70>}
     {limitMax=<180 180 0>, limitMin=<-180 -180 -70>}
     {limitMax=<180 180 0>, limitMin=<-180 -180 -70>}
     {limitMax=<180 180 0>, limitMin=<-180 -180 -70>}
    ]
    rotationOffsets [<0 0 0>, <0 0 0>, <0 0 0>, <0 0 0>, <0 0 0>, <0 0 0>, <0 0 0>]
    trailBone {boneName=dyn_cape_r_07}
    trailRelaxationSpeed {editorCurveData {keys [<at 0 15 RCIM_Linear>,<at 0.5289 13.1838 RCIM_Linear>,<at 1 6 RCIM_Linear>]}}
   }
   Save cached pose 'trailCache'(Pose=Blend Poses by bool(
     BlendPose_0=Component To Local(ComponentPose=@n15)
     BlendPose_1=Use cached pose 'RBAN_cache'()
     bActiveValue=Get trailControllerConditions() -> (trailControllerConditions:bool)
    ) {blendType=Linear}) {cacheName=trailCache}
   n16 = RigidBody!(ComponentPose=Local To Component(LocalPose=Use cached pose 'Start_Cache'()), Alpha, OverrideWorldGravity:Vector=Get GravityOverride()) -> (Pose) {
    bClampLinearTranslationLimitToRefPose=true
    bOverrideWorldGravity=true
    overridePhysicsAsset=&/M_MED_ArcticIceBlue_Pack_Physics
    simulationSpace=BaseBoneSpace
    alphaBoolBlend {blendOption=Linear}
    baseBoneRef {boneName=spine_02}
    componentAppliedLinearAccClamp=<1000 1000 1000>
    componentLinearAccScale=<1.5 0.2 0.2>
    componentLinearVelScale=<0.22 0.212 0.22>
    overrideWorldGravity {z=-980}
    simSpaceSettings {
     maxAngularAcceleration=20
     maxAngularVelocity=1
     maxLinearAcceleration=300
     maxLinearVelocity=150
     velocityScaleZ=0
     externalLinearDragV=<0.2 0.2>
    }
   }
   n17 = Component To Local(ComponentPose=@n16)
   n18 = RigidBody(ComponentPose=Local To Component(LocalPose=Use cached pose 'Start_Cache'()), OverrideWorldGravity=Get GravityOverride()) {
    bClampLinearTranslationLimitToRefPose=true
    bOverrideWorldGravity=true
    overridePhysicsAsset=&/M_MED_ArcticIceBlue_skydive_Pack_Physics
    simulationSpace=BaseBoneSpace
    alphaBoolBlend {blendOption=Linear}
    baseBoneRef {boneName=spine_02}
    componentAppliedLinearAccClamp=<1000 1000 1000>
    componentLinearAccScale=<1.5 0.2 0.2>
    componentLinearVelScale=<0.22 0.212 0.22>
    overrideWorldGravity {z=-980}
    simSpaceSettings {
     maxAngularAcceleration=20
     maxAngularVelocity=1
     maxLinearAcceleration=300
     maxLinearVelocity=150
     velocityScaleZ=0
     worldAlpha=1
     externalLinearDragV=<0.2 0.2>
    }
   }
   n19 = Component To Local(ComponentPose=@n18)
   n20 = Get bIsSkyDiving()
   n21 = Get isSkydiveFloating() -> (isSkydiveFloating:float)
   n22 = "Transform (Modify) Bone - Bone: dyn_L_frontPull"(
    ComponentPose=Local To Component(LocalPose=Blend Poses by bool(BlendPose_0=@n19, BlendPose_1=@n17, bActiveValue=@n20) {blendType=Linear})
    Alpha=@n21
    Translation="0,0,0"
    Rotation="10,-5,0"
    Scale="1,1.2,1"
   ) {
    rotationMode=BMM_Additive
    scaleMode=BMM_Additive
    translationMode=BMM_Additive
    alphaBoolBlend {blendOption=Linear}
    alphaScaleBiasClamp {bInterpResult=true, interpSpeedDecreasing=30, interpSpeedIncreasing=30}
    boneToModify {boneName=dyn_L_frontPull}
    rotation {pitch=10, yaw=-5}
    scale {y=1.2}
   }
   n23 = "Transform (Modify) Bone - Bone: dyn_R_frontPull"(ComponentPose=@n22, Alpha=@n21, Translation="0,0,0", Rotation="-10,5,0", Scale="1,1.2,1") {
    rotationMode=BMM_Additive
    scaleMode=BMM_Additive
    translationMode=BMM_Additive
    alphaBoolBlend {blendOption=Linear}
    alphaScaleBiasClamp {bInterpResult=true, interpSpeedDecreasing=30, interpSpeedIncreasing=30}
    boneToModify {boneName=dyn_R_frontPull}
    rotation {pitch=-10, yaw=5}
    scale {y=1.2}
   }
   n24 = Get isSkydiveDiveMode() -> (isSkydiveDiveMode:float)
   n25 = Control Rig!(Source=Component To Local(ComponentPose=@n23), Alpha, is_skydive_dive:double=@n24) -> (Pose) {
    alphaBoolBlend {blendOption=Linear}
    controlRigAssetReference {blueprintRigClass=&/M_MED_ArcticIceBlue_Pack_skydive_CtrlRig.M_MED_ArcticIceBlue_Pack_skydive_CtrlRig_C}
   }
   n26 = "Constraint - dyn_cape_02 (PP)"!(ComponentPose=Local To Component(LocalPose=@n25), Alpha=@n24, ConstraintWeights_0=0.5, ConstraintWeights_1:float=0.5) -> (Pose) {
    alphaBoolBlend {blendOption=Linear}
    alphaScaleBiasClamp {bInterpResult=true, interpSpeedDecreasing=40, interpSpeedIncreasing=40}
    boneToModify {boneName=dyn_cape_02}
    constraintSetup [{targetBone {boneName=dyn_cape_l_02}}, {targetBone {boneName=dyn_cape_r_02}}]
    constraintWeights [1,1]
   }
   n27 = "Constraint - dyn_cape_03 (PP)"!(ComponentPose=@n26, Alpha=@n24, ConstraintWeights_0=0.5, ConstraintWeights_1:float=0.5) -> (Pose) {
    alphaBoolBlend {blendOption=Linear}
    alphaScaleBiasClamp {bInterpResult=true, interpSpeedDecreasing=40, interpSpeedIncreasing=40}
    boneToModify {boneName=dyn_cape_03}
    constraintSetup [{targetBone {boneName=dyn_cape_l_03}}, {targetBone {boneName=dyn_cape_r_03}}]
    constraintWeights [1,1]
   }
   n28 = "Constraint - dyn_cape_04 (PP)"!(ComponentPose=@n27, Alpha=@n24, ConstraintWeights_0=0.5, ConstraintWeights_1:float=0.5) -> (Pose) {
    alphaBoolBlend {blendOption=Linear}
    alphaScaleBiasClamp {bInterpResult=true, interpSpeedDecreasing=40, interpSpeedIncreasing=40}
    boneToModify {boneName=dyn_cape_04}
    constraintSetup [{targetBone {boneName=dyn_cape_l_04}}, {targetBone {boneName=dyn_cape_r_04}}]
    constraintWeights [1,1]
   }
   n29 = "Constraint - dyn_cape_05 (PP)"!(ComponentPose=@n28, Alpha=@n24, ConstraintWeights_0=0.5, ConstraintWeights_1:float=0.5) -> (Pose) {
    alphaBoolBlend {blendOption=Linear}
    alphaScaleBiasClamp {bInterpResult=true, interpSpeedDecreasing=40, interpSpeedIncreasing=40}
    boneToModify {boneName=dyn_cape_05}
    constraintSetup [{targetBone {boneName=dyn_cape_l_05}}, {targetBone {boneName=dyn_cape_r_05}}]
    constraintWeights [1,1]
   }
   n30 = "Constraint - dyn_cape_06 (PP)"!(ComponentPose=@n29, Alpha=@n24, ConstraintWeights_0=0.5, ConstraintWeights_1:float=0.5) -> (Pose) {
    alphaBoolBlend {blendOption=Linear}
    alphaScaleBiasClamp {bInterpResult=true, interpSpeedDecreasing=40, interpSpeedIncreasing=40}
    boneToModify {boneName=dyn_cape_06}
    constraintSetup [{targetBone {boneName=dyn_cape_l_06}}, {targetBone {boneName=dyn_cape_r_06}}]
    constraintWeights [1,1]
   }
   n31 = "Constraint - dyn_cape_07 (PP)"!(ComponentPose=@n30, Alpha=@n24, ConstraintWeights_0=0.5, ConstraintWeights_1:float=0.5) -> (Pose) {
    alphaBoolBlend {blendOption=Linear}
    alphaScaleBiasClamp {bInterpResult=true, interpSpeedDecreasing=40, interpSpeedIncreasing=40}
    boneToModify {boneName=dyn_cape_07}
    constraintSetup [{targetBone {boneName=dyn_cape_l_07}}, {targetBone {boneName=dyn_cape_r_07}}]
    constraintWeights [1,1]
   }
   n32 = Get isTacticalSprinting() -> (isTacticalSprinting:float)
   n33 = "Transform (Modify) Bone - Bone: dyn_L_capeSleeve_follow"!(ComponentPose=@n31, Alpha=@n32, Translation="0,-20,0", Rotation="0,5,15") -> (Pose) {
    rotationMode=BMM_Additive
    translationMode=BMM_Additive
    alphaBoolBlend {blendOption=Linear}
    alphaScaleBiasClamp {bInterpResult=true, interpSpeedDecreasing=40, interpSpeedIncreasing=40}
    boneToModify {boneName=dyn_L_capeSleeve_follow}
    rotation {roll=15, yaw=5}
    translation {y=-20}
   }
   n34 = "Transform (Modify) Bone - Bone: dyn_R_capeSleeve_follow"!(ComponentPose=@n33, Alpha=@n32, Translation="0,-20,0", Rotation="0.000,5.000,15.000") -> (Pose) {
    rotationMode=BMM_Additive
    translationMode=BMM_Additive
    alphaBoolBlend {blendOption=Linear}
    alphaScaleBiasClamp {bInterpResult=true, interpSpeedDecreasing=40, interpSpeedIncreasing=40}
    boneToModify {boneName=dyn_R_capeSleeve_follow}
    rotation {roll=15, yaw=5}
    translation {y=-20}
   }
   n35 = "Transform (Modify) Bone - Bone: dyn_cape_l_MSTR"!(ComponentPose=@n34, Alpha=@n32, Translation="0,-10,0", Rotation="0,0,0") -> (Pose) {
    rotationMode=BMM_Additive
    translationMode=BMM_Additive
    alphaBoolBlend {blendOption=Linear}
    alphaScaleBiasClamp {bInterpResult=true, interpSpeedDecreasing=40, interpSpeedIncreasing=40}
    boneToModify {boneName=dyn_cape_l_MSTR}
    translation {y=-10}
   }
   n36 = "Transform (Modify) Bone - Bone: dyn_cape_r_MSTR"!(ComponentPose=@n35, Alpha=@n32, Translation="0,-10,0", Rotation="0,0,0") -> (Pose) {
    rotationMode=BMM_Additive
    translationMode=BMM_Additive
    alphaBoolBlend {blendOption=Linear}
    alphaScaleBiasClamp {bInterpResult=true, interpSpeedDecreasing=40, interpSpeedIncreasing=40}
    boneToModify {boneName=dyn_cape_r_MSTR}
    translation {y=-10}
   }
   Save cached pose 'RBAN_cache'(Pose=Component To Local(ComponentPose=@n36)) {cacheName=RBAN_cache}
  ]}
 EventGraph {nodes [
   n1 = Event BlueprintUpdateAnimation()
   > n2 = Sequence!(execute) -> (then_0, then_1, then_2:Exec, then_3:Exec, then_4:Exec, then_5:Exec)
   > n3 = Branch(Condition=Get bIsSkyDiving())
   > Set SkydivingAlpha(SkydivingAlpha:Rotator="0,0,30") -> (Output_Get:Rotator)
   Set SkydivingAlpha(execute=@n3.else, SkydivingAlpha="0,0,5")
   n4 = PlayerMovement(execute=@n2.then_1, AnimInstance=Self-Reference(), XDotProductCoefficient=1, YDotProductCoefficient=1, ZDotProductCoefficient=1)
   > n5 = Branch(Condition=AND Boolean(A=Get bIsTargeting(), B=Get bIsCrouching()))
   n6 = MapRangeClamped(
    Value=float + float(A="Absolute (Float)"(A=@n4), B=MapRangeClamped(Value=@n4.YMovement, InRangeA=-1, InRangeB=1, OutRangeA=0.3)) {is=title K2Node_CommutativeAssociativeBinaryOperator}
    InRangeA=0
    InRangeB=1.65
   )
   > Set BwdMovement(BwdMovement:double=float * float(A=@n6, B=0.25) {is=title K2Node_PromotableOperator}) -> (Output_Get:double)
   Set BwdMovement(execute=@n5.else, BwdMovement=@n6)
   n7 = Branch(execute=@n2.then_2, Condition=Get bIsTacticalSprinting())
   > n8 = Branch(Condition="float > float"(A=Get PawnVelocityZ(), B=50) {is=title K2Node_CallFunction})
   > Set BwdMovement(BwdMovement=float * float(A=Get BwdMovement(), B=0) {is=title K2Node_CommutativeAssociativeBinaryOperator})
   Set BwdMovement(execute=[@n8.else, @n7.else], BwdMovement=float * float(A=Get BwdMovement(), B=1))
   n9 = Set Current LOD(execute=@n2.then_3, Current LOD:int=Get CurrentLOD()) -> (Output_Get:int=0)
   > Set lowLODEnabled(lowLODEnabled:bool="integer >= integer"(A=@n9, B=2)) -> (Output_Get:bool)
   n10 = Get bRigidBodyEnabled()
   > Set trailControllerConditions(trailControllerConditions:bool=OR Boolean(A=Get lowLODEnabled() -> (lowLODEnabled:bool), B=NOT Boolean(A=@n10))) -> (Output_Get:bool)
   n11 = Gravity Override State Function DEPRECATED(
    execute=@n2.then_4
    CharacterPartAnimInstance=Self-Reference()
    GameTimeInSeconds=GetGameTimeInSeconds()
    GravityOverrideParams=Get Gravity Override Params()
   )
   > n12 = Set GravityOverride(GravityOverride=@n11)
   > Set GravityOverride(GravityOverride=SelectVector(A="0,0,-800", B=@n12, bPickA=Get isJumpFalling() -> (isJumpFalling:bool)))
   > Set isSkydiveFloating(isSkydiveFloating:float="To Float (Boolean)"(InBool=Get bIsSkydiveFloating())) -> (Output_Get:float)
   > Set isTacticalSprinting(isTacticalSprinting:float="To Float (Boolean)"(InBool=Get bIsTacticalSprinting())) -> (Output_Get:float)
   > Set isSkydiving(isSkydiving:float="To Float (Boolean)"(InBool=Get bIsSkyDiving())) -> (Output_Get:float)
   > Set isSkydiveDiveMode(isSkydiveDiveMode:float="To Float (Boolean)"(InBool=Get bIsSkydiveDiveMode())) -> (Output_Get:float)
   > Set isSwimming(isSwimming:float="To Float (Boolean)"(InBool=Get bIsSwimSprinting() -> (bIsSwimSprinting:bool))) -> (Output_Get:float)
   n13 = GetOwnerSkeletalMeshComponent()
   n14 = BreakRotator(InRot=BreakTransform(InTransform=GetDeltaTransformFromRefPose(self=@n13, BoneName=spine_05, BaseName=upperarm_l)).Rotation)
   n15 = MapRangeClamped(Value=@n14.Yaw, InRangeA=20, InRangeB=60)
   Set L armUp(
    execute=@n2.then_5
    L armUp:double="Max (Float)"(A=Get isSkydiving() -> (isSkydiving:float), B=float * float(A=@n15, B=0.6) {is=title K2Node_PromotableOperator})
   ) -> (Output_Get:double)
   n16 = "Max (Float)"(A=MapRangeClamped(Value=@n14.Pitch, InRangeA=0, InRangeB=-50), B=@n15)
   n17 = Get isSkydiveFloating()
   > Set L armForward(L armForward:double="Max (Float)"!(A=float * float(A=@n16, B=0.6), B=@n17, C:double) -> (ReturnValue)) -> (Output_Get:double)
   n18 = BreakRotator(InRot=BreakTransform(InTransform=GetDeltaTransformFromRefPose(self=@n13, BoneName=spine_05, BaseName=upperarm_r)).Rotation)
   n19 = MapRangeClamped(Value=@n18.Yaw, InRangeA=20, InRangeB=60)
   > Set R armUp(R armUp:double="Max (Float)"!(A=Get isSkydiving(), B=float * float(A=@n19, B=0.6)) -> (ReturnValue)) -> (Output_Get:double)
   n20 = "Max (Float)"(A=MapRangeClamped(Value=@n18.Pitch, InRangeA=0, InRangeB=50), B=@n19)
   > Set R armForward(R armForward:double="Max (Float)"!(A=float * float(A=@n20, B=0.6), B=@n17, C:double) -> (ReturnValue)) -> (Output_Get:double)
   Event BlueprintInitializeAnimation()
   n21 = GetOwnerSkeletalMeshComponent()
   > Is Valid(InputObject=@n21)
   > Set Skeletal_Mesh(Skeletal_Mesh:Skeletal Mesh Component Object Reference=@n21) -> (Output_Get:Skeletal Mesh Component Object Reference)
   Event OnEndFalling()
   > Set isJumpFalling(isJumpFalling:bool=false) -> (Output_Get:bool)
   Event OnBeginFalling()
   > Set isJumpFalling(isJumpFalling=true)
   Event OnBeginJumping()
   > Set isJumpFalling(isJumpFalling=true)
   n22 = Branch(Condition=Get bClothEnabled())
   > Set turnRBANoff(turnRBANoff:bool=false) -> (Output_Get:bool)
   Set turnRBANoff(execute=@n22.else, turnRBANoff=true)
   n23 = Event BlueprintProcessAnimInputEvent()
   n24 = Branch(Condition=IsNamed(self=@n23.AnimInputEvent, InEventName=aie.Player.Zipline.Entered))
   > Delay(Duration=0.1)
   n25 = Get ChaosClothingInteractor() -> (ChaosClothingInteractor:Chaos Clothing Interactor Object Reference)
   > ResetAndTeleport!(execute, self:Chaos Clothing Interactor Object Reference=@n25, bReset:bool=true, bTeleport:bool=true) -> (then)
   Branch(execute=@n24.else, Condition=IsNamed(self=@n23.AnimInputEvent, InEventName=aie.Player.Zipline.Exited))
   > Set resetClothLandingFromZipline(resetClothLandingFromZipline:bool=true) -> (Output_Get:bool)
   Branch(Condition=Get resetClothLandingFromZipline() -> (resetClothLandingFromZipline:bool))
   > ResetAndTeleport(self=@n25, bReset=true, bTeleport=true)
   > Set resetClothLandingFromZipline(resetClothLandingFromZipline=false)
   n26 = GetClothingSimulationInteractor!(self:Skeletal Mesh Component Object Reference=GetOwningComponent(), ClothingAssetName:Name=None) -> (ReturnValue:Clothing Simulation Interactor Object Reference)
   Is Valid(InputObject=@n26)
   n27 = MakeLiteralString(Value=M_MED_ArcticIceBlue_Pack_Clothing_0)
   > n28 = Cast To ChaosClothingInteractor!(
    execute
    Object=GetClothingInteractor!(self:Clothing Simulation Interactor Object Reference=@n26, ClothingAssetName:Name=String To Name(InString=@n27)) -> (ReturnValue:Clothing Interactor Object Reference)
   ) -> (then, CastFailed, AsChaos Clothing Interactor:Chaos Clothing Interactor Object Reference) {is=title K2Node_DynamicCast}
   > Set ChaosClothingInteractor(ChaosClothingInteractor:Chaos Clothing Interactor Object Reference=@n28) -> (Output_Get:Chaos Clothing Interactor Object Reference)
   Break Cloth Setting Struct!(ClothSettingStruct:"Cloth Setting Struct Structure (by ref)"=Cloth Controller Function DEPRECATED!(
     execute
     CharacterPartAnimInstance:Custom Character Part Anim Instance Object Reference=Self-Reference()
     ClothingAssetName:String=M_MED_ArcticIceBlue_Pack_Clothing_0
     Cloth Paramaters:Cloth Params Struct Structure=Get Cloth Paramaters_0() -> (Cloth Paramaters_0:Cloth Params Struct Structure)
     Cloth Core Settings:Cloth Core Setting Struct Structure=Get Cloth Core Settings_0() -> (Cloth Core Settings_0:Cloth Core Setting Struct Structure)
     ClothUpAxis:Vector="0,0,1"
     OnGround_Moving_SpeedLimit:double=2
     OnGround_Moving_AngularVel_Max:double=0.05
     OnGround_Moving_FictitiousVel_Max:double=0.6
     LinVelScale_FwdBwd_Clamp:double=0
     LinVelScale_UpDwn_Clamp:double=0.7
     DeltaTime:double=@n1.DeltaTimeX
     GameTimeInSeconds:double=GetGameTimeInSeconds()
     WindGustsParams:Wind Gust Params Struct Structure=Get Wind Gusts Params() -> (Wind Gusts Params:Wind Gust Params Struct Structure)
     WindGustLocalSettings:Wind Gust Local Setting Struct Structure=Get Wind Gust Local Settings() -> (Wind Gust Local Settings:Wind Gust Local Setting Struct Structure)
    ) -> (
     then
     CurrentStateOut:String
     Cloth Settings Out:Cloth Setting Struct Structure
     Cloth Core Settings Out:Cloth Core Setting Struct Structure
     WindGustSettings Out:Wind Gust Setting Struct Structure
     WindGustLocalSettings Out:Wind Gust Local Setting Struct Structure
     GustStrength - WindVelocityOut:Vector
     LinearVelVector-DELETE LATER:Vector
    ).Cloth Settings Out) -> (
    LinearVelocityScale_7_#1:Vector=~
    AngularVelocityScale_6_#2:float=~
    FictitiousAngularScale_10_#3:float=~
    GravityOverride_15_#4:Vector=~
    DragCoefficient_23_#5:Vector 2D Structure
    LiftCoefficient_24_#6:Vector 2D Structure
    JointName_27_#7:Name
   ) {is=title K2Node_BreakStruct}
   AND Boolean(A=NOT Boolean(A=Get bClothEnabled()), B=false)
   Sequence!(execute) -> (then_0, then_1)
   Branch(Condition=Get bClothEnabled())
   Branch(Condition=NOT Boolean(A=Get bClothEnabled()))
   Branch(Condition=NOT Boolean(A=Get bClothEnabled()))
   Get isTacticalSprinting()
  ]}
]
variables [
 BwdMovement {type=number}
 ChaosClothingInteractor {type=ChaosClothingInteractor, default=null}
 Cloth Core Settings_0 {
  type=ClothCoreSettingStruct
  default {
   collisionThickness=1.7
   dampingCoefficient=0.2
   frictionCoefficient=0.4
   localDampingCoefficient=0.1
   numberIterations=2
   numberSubsteps=1
   selfCollisionThickness=0
   animDriveDamping=<0 1>
   animDriveStiffness=<0 1>
   areaStiffness=<1 1>
   bendingStiffness=<1 1>
   edgeStiffness=<1 1>
   tetherStiffness=<0.6 0.8>
  }
 }
 Cloth Paramaters_0 {
  type=ClothParamsStruct
  default {
   dBNO {
    angularVelocityScale=0.6
    fictitiousAngularScale=0.1
    jointName=pelvis
    dragCoefficient=<0.2 1>
    gravityOverride=<-980 400 0>
    liftCoefficient=<0 0.5>
    linearVelocityScale=<0.3 0.3 0.1>
   }
   dBNO_Carried {
    angularVelocityScale=0.6
    fictitiousAngularScale=0.1
    jointName=pelvis
    dragCoefficient=<0.2 1>
    gravityOverride=<-980 -200 0>
    liftCoefficient=<0 0.5>
    linearVelocityScale=<0.3 0.3 0.1>
   }
   dBNO_Impulse {
    angularVelocityScale=0
    fictitiousAngularScale=0
    jointName=pelvis
    dragCoefficient=<0 0.2>
    gravityOverride=<-3000 -2000 0>
    liftCoefficient=<0 0.1>
    linearVelocityScale=<0.3 0.1 0.3>
   }
   dBNO_RollBwd {
    angularVelocityScale=0
    fictitiousAngularScale=0
    jointName=pelvis
    dragCoefficient=<0 0.2>
    gravityOverride=<-2000 0 0>
    liftCoefficient=<0 0.1>
    linearVelocityScale=<0.3 0.1 0.3>
   }
   dBNO_RollFwd {
    angularVelocityScale=0
    fictitiousAngularScale=0
    jointName=pelvis
    dragCoefficient=<0 0.2>
    gravityOverride=<-980 0 0>
    liftCoefficient=<0 0.1>
    linearVelocityScale=<0.3 0.1 0.3>
   }
   dBNO_Tumbling {
    angularVelocityScale=0
    fictitiousAngularScale=0
    jointName=pelvis
    dragCoefficient=<0 0.2>
    gravityOverride=<-2500 -300 0>
    liftCoefficient=<0 0.1>
    linearVelocityScale=<0.3 0.1 0.3>
   }
   emoteOrMelee {
    angularVelocityScale=0.6
    fictitiousAngularScale=0.1
    jointName=root
    dragCoefficient=<2.5 2.5>
    gravityOverride=<0 -700 -980>
    liftCoefficient=<0.5 0.5>
    linearVelocityScale=<0.6 0.6 0.5>
   }
   front_End {
    angularVelocityScale=0.05
    fictitiousAngularScale=0.6
    jointName=root
    dragCoefficient=<0.2 0.2>
    gravityOverride=<0 -200 -980>
    liftCoefficient=<0.1 0.1>
    linearVelocityScale=<0.1 0.1 0.1>
   }
   onGround_Crouch_Moving {
    angularVelocityScale=0.1
    fictitiousAngularScale=0.3
    jointName=root
    dragCoefficient=<2 2>
    gravityOverride=<0 0 -980>
    liftCoefficient=<0.5 0.5>
    linearVelocityScale=<0.6 0.6 0.5>
   }
   onGround_Falling {
    angularVelocityScale=0.6
    fictitiousAngularScale=0
    jointName=root
    dragCoefficient=<0 0.2>
    gravityOverride=<0 -500 -980>
    liftCoefficient=<0 0.1>
    linearVelocityScale=<2 1 0.3>
   }
   onGround_Moving {
    angularVelocityScale=0.05
    fictitiousAngularScale=0.2
    jointName=root
    dragCoefficient=<2.5 2.5>
    gravityOverride=<0 0 -1800>
    liftCoefficient=<0.5 0.5>
    linearVelocityScale=<0.6 0.6 0.5>
   }
   onGround_Standing {
    angularVelocityScale=0.05
    fictitiousAngularScale=0.2
    jointName=root
    dragCoefficient=<2.5 2.5>
    gravityOverride=<0 0 -980>
    liftCoefficient=<0.5 0.5>
    linearVelocityScale=<0.6 0.6 0.5>
   }
   rail_Grinding {
    angularVelocityScale=0.2
    fictitiousAngularScale=0
    jointName=root
    dragCoefficient=<0.1 0.2>
    gravityOverride=<50 -500 -400>
    liftCoefficient=<0 0.5>
    linearVelocityScale=<0.1 0.1 0.1>
   }
   riding_Moving {
    angularVelocityScale=0.05
    fictitiousAngularScale=1
    jointName=root
    dragCoefficient=<0.5 0.5>
    gravityOverride=<0 0 -980>
    liftCoefficient=<0.5 0.5>
    linearVelocityScale=<0.4 0.6 0.3>
   }
   riding_Moving_Targeting {
    angularVelocityScale=0.05
    fictitiousAngularScale=0.7
    jointName=root
    dragCoefficient=<0.5 0.5>
    gravityOverride=<0 0 -980>
    liftCoefficient=<0.5 0.5>
    linearVelocityScale=<0.2 0.6 0.2>
   }
   riding_Standing {
    angularVelocityScale=0.05
    fictitiousAngularScale=1
    jointName=root
    dragCoefficient=<0.5 0.5>
    gravityOverride=<0 -450 -980>
    liftCoefficient=<0.5 0.5>
    linearVelocityScale=<0.6 0.6 0.3>
   }
   skydive_DivingDown {
    angularVelocityScale=0.1
    fictitiousAngularScale=0
    jointName=pelvis
    dragCoefficient=<0.3 0.3>
    gravityOverride=<-450 -100 0>
    liftCoefficient=<0.9 0.9>
    linearVelocityScale=<0.1 0.1 0.1>
   }
   skydive_FloatingDown {
    angularVelocityScale=0.1
    fictitiousAngularScale=0
    jointName=root
    dragCoefficient=<0.3 0.3>
    gravityOverride=<0 -800 200>
    liftCoefficient=<1.3 1.3>
    linearVelocityScale=<0.2 0.2 0.1>
   }
   skydive_GoingUp {
    angularVelocityScale=0.1
    fictitiousAngularScale=0
    jointName=root
    dragCoefficient=<0.01 0.01>
    gravityOverride=<0 -2000 400>
    liftCoefficient=<0.05 0.05>
    linearVelocityScale=<0.1 0.1 0.01>
   }
   skydive_Parachuting {
    angularVelocityScale=0.6
    fictitiousAngularScale=0.1
    jointName=root
    dragCoefficient=<0.8 0.8>
    gravityOverride=<0 -600 -500>
    liftCoefficient=<0.8 0.8>
    linearVelocityScale=<0.2 0.2 0.3>
   }
   skydive_Surfing {
    angularVelocityScale=0.6
    fictitiousAngularScale=0.1
    jointName=root
    dragCoefficient=<1.5 1.5>
    gravityOverride=<0 -600 -500>
    liftCoefficient=<1 1>
    linearVelocityScale=<0.1 0.1 0.3>
   }
   swimming {
    angularVelocityScale=0.6
    fictitiousAngularScale=0
    jointName=root
    dragCoefficient=<2.5 2.5>
    gravityOverride=<0 0 0>
    liftCoefficient=<0.1 0.1>
    linearVelocityScale=<0.6 0.6 0.5>
   }
   tactical_Sprint {
    angularVelocityScale=0.6
    fictitiousAngularScale=0.1
    jointName=root
    dragCoefficient=<2 2>
    gravityOverride=<0 0 -980>
    liftCoefficient=<0.5 0.5>
    linearVelocityScale=<0.6 0.6 0.5>
   }
   ziplining {
    angularVelocityScale=0.6
    fictitiousAngularScale=0
    jointName=root
    dragCoefficient=<0 1.5>
    gravityOverride=<0 0 -980>
    liftCoefficient=<0 0.3>
    linearVelocityScale=<0.4 0.4 0.5>
   }
  }
 }
 Current LOD {type=integer}
 Gravity Override Params {
  type=GravityOverrideParamsStruct
  default {
   onGround_Falling {gravityOverride=<0 0 -200>}
   skydive_DivingDown {gravityOverride=<-1500 0 1000>}
   skydive_FloatingDown {gravityOverride=<0 0 100>}
   skydive_GoingUp {gravityOverride=<0 0 100>}
   skydive_Parachuting {gravityOverride=<0 0 -980>}
   tactical_Sprint {gravityOverride=<0 0 -2000>}
  }
 }
 GravityOverride {type=Vector}
 isJumpFalling {type=boolean}
 isSkydiveDiveMode {type=number}
 isSkydiveFloating {type=number}
 isSkydiving {type=number}
 isSwimming {type=number}
 isTacticalSprinting {type=number}
 L armForward {type=number}
 L armUp {type=number}
 lowLODEnabled {type=boolean}
 R armForward {type=number}
 R armUp {type=number}
 resetClothLandingFromZipline {type=boolean}
 Skeletal_Mesh {type=SkeletalMeshComponent}
 SkydivingAlpha {type=Rotator}
 trailControllerConditions {type=boolean}
 turnRBANoff {type=boolean}
 Wind Gust Local Settings {type=WindGustLocalSettingStruct, default {elapsedTime=0, finalGustStrengthLocal=0, gustDuration=0, gustStrength=0, timeBetweenGusts=0}}
 Wind Gusts Params {
  type=WindGustParamsStruct
  default {
   dBNO {
    maxDownInterpSpeed=5
    maxGustDuration=2
    maxGustStrength=7
    maxTimeBetweenGusts=7
    maxUpInterpSpeed=5
    minDownInterpSpeed=2
    minGustDuration=0.5
    minGustStrength=3
    minTimeBetweenGusts=0
    minUpInterpSpeed=2
    windTurbulenceScalarA=0.1
    windTurbulenceScalarB=1
    windAmplitude=<0 0 0>
    windFrequency=<0 0 0>
    windOffset=<0 0 0>
   }
   dBNO_Carried {
    maxDownInterpSpeed=5
    maxGustDuration=2
    maxGustStrength=7
    maxTimeBetweenGusts=7
    maxUpInterpSpeed=5
    minDownInterpSpeed=2
    minGustDuration=0.5
    minGustStrength=3
    minTimeBetweenGusts=0
    minUpInterpSpeed=2
    windTurbulenceScalarA=0.1
    windTurbulenceScalarB=1
    windAmplitude=<0 0 0>
    windFrequency=<0 0 0>
    windOffset=<0 0 0>
   }
   dBNO_Impulse {
    maxDownInterpSpeed=5
    maxGustDuration=2
    maxGustStrength=7
    maxTimeBetweenGusts=7
    maxUpInterpSpeed=5
    minDownInterpSpeed=2
    minGustDuration=0.5
    minGustStrength=3
    minTimeBetweenGusts=0
    minUpInterpSpeed=2
    windTurbulenceScalarA=0.1
    windTurbulenceScalarB=1
    windAmplitude=<0 0 0>
    windFrequency=<0 0 0>
    windOffset=<0 0 0>
   }
   dBNO_RollBwd {
    maxDownInterpSpeed=5
    maxGustDuration=2
    maxGustStrength=7
    maxTimeBetweenGusts=7
    maxUpInterpSpeed=5
    minDownInterpSpeed=2
    minGustDuration=0.5
    minGustStrength=3
    minTimeBetweenGusts=0
    minUpInterpSpeed=2
    windTurbulenceScalarA=0.1
    windTurbulenceScalarB=1
    windAmplitude=<0 0 0>
    windFrequency=<0 0 0>
    windOffset=<0 0 0>
   }
   dBNO_RollFwd {
    maxDownInterpSpeed=5
    maxGustDuration=2
    maxGustStrength=7
    maxTimeBetweenGusts=7
    maxUpInterpSpeed=5
    minDownInterpSpeed=2
    minGustDuration=0.5
    minGustStrength=3
    minTimeBetweenGusts=0
    minUpInterpSpeed=2
    windTurbulenceScalarA=0.1
    windTurbulenceScalarB=1
    windAmplitude=<0 0 0>
    windFrequency=<0 0 0>
    windOffset=<0 0 0>
   }
   dBNO_Tumbling {
    maxDownInterpSpeed=5
    maxGustDuration=2
    maxGustStrength=7
    maxTimeBetweenGusts=7
    maxUpInterpSpeed=5
    minDownInterpSpeed=2
    minGustDuration=0.5
    minGustStrength=3
    minTimeBetweenGusts=0
    minUpInterpSpeed=2
    windTurbulenceScalarA=0.1
    windTurbulenceScalarB=1
    windAmplitude=<0 0 0>
    windFrequency=<0 0 0>
    windOffset=<0 0 0>
   }
   emoteOrMelee {
    maxDownInterpSpeed=5
    maxGustDuration=2
    maxGustStrength=7
    maxTimeBetweenGusts=7
    maxUpInterpSpeed=5
    minDownInterpSpeed=2
    minGustDuration=0.5
    minGustStrength=3
    minTimeBetweenGusts=0
    minUpInterpSpeed=2
    windTurbulenceScalarA=0.1
    windTurbulenceScalarB=1
    windAmplitude=<0 0 0>
    windFrequency=<0 0 0>
    windOffset=<0 0 0>
   }
   front_End {
    maxDownInterpSpeed=5
    maxGustDuration=2
    maxGustStrength=7
    maxTimeBetweenGusts=7
    maxUpInterpSpeed=5
    minDownInterpSpeed=2
    minGustDuration=0.5
    minGustStrength=3
    minTimeBetweenGusts=0
    minUpInterpSpeed=2
    windTurbulenceScalarA=0.1
    windTurbulenceScalarB=1
    windAmplitude=<0 0 0>
    windFrequency=<0 0 0>
    windOffset=<0 0 0>
   }
   onGround_Crouch_Moving {
    maxDownInterpSpeed=5
    maxGustDuration=2
    maxGustStrength=7
    maxTimeBetweenGusts=7
    maxUpInterpSpeed=5
    minDownInterpSpeed=2
    minGustDuration=0.5
    minGustStrength=3
    minTimeBetweenGusts=0
    minUpInterpSpeed=2
    windTurbulenceScalarA=0.1
    windTurbulenceScalarB=1
    windAmplitude=<0 0 0>
    windFrequency=<0 0 0>
    windOffset=<0 0 0>
   }
   onGround_Falling {
    maxDownInterpSpeed=5
    maxGustDuration=2
    maxGustStrength=7
    maxTimeBetweenGusts=7
    maxUpInterpSpeed=5
    minDownInterpSpeed=2
    minGustDuration=0.5
    minGustStrength=3
    minTimeBetweenGusts=0
    minUpInterpSpeed=2
    windTurbulenceScalarA=0.1
    windTurbulenceScalarB=1
    windAmplitude=<0 0 0>
    windFrequency=<0 0 0>
    windOffset=<0 0 0>
   }
   onGround_Moving {
    maxDownInterpSpeed=5
    maxGustDuration=2
    maxGustStrength=7
    maxTimeBetweenGusts=7
    maxUpInterpSpeed=5
    minDownInterpSpeed=2
    minGustDuration=0.5
    minGustStrength=3
    minTimeBetweenGusts=0
    minUpInterpSpeed=2
    windTurbulenceScalarA=0.1
    windTurbulenceScalarB=1
    windAmplitude=<0 0 0>
    windFrequency=<0 0 0>
    windOffset=<0 0 0>
   }
   onGround_Standing {
    maxDownInterpSpeed=5
    maxGustDuration=2
    maxGustStrength=7
    maxTimeBetweenGusts=7
    maxUpInterpSpeed=5
    minDownInterpSpeed=2
    minGustDuration=0.5
    minGustStrength=3
    minTimeBetweenGusts=0
    minUpInterpSpeed=2
    windTurbulenceScalarA=0.1
    windTurbulenceScalarB=1
    windAmplitude=<0 0 0>
    windFrequency=<0 0 0>
    windOffset=<0 0 0>
   }
   rail_Grinding {
    maxDownInterpSpeed=5
    maxGustDuration=8
    maxGustStrength=7
    maxTimeBetweenGusts=8
    maxUpInterpSpeed=5
    minDownInterpSpeed=2
    minGustDuration=0.2
    minGustStrength=3
    minTimeBetweenGusts=1
    minUpInterpSpeed=2
    windTurbulenceScalarA=0.1
    windTurbulenceScalarB=0.8
    windAmplitude=<20 65 42>
    windFrequency=<5 6 8.6>
    windOffset=<50 -500 -400>
   }
   riding_Moving {
    maxDownInterpSpeed=5
    maxGustDuration=2
    maxGustStrength=7
    maxTimeBetweenGusts=7
    maxUpInterpSpeed=5
    minDownInterpSpeed=2
    minGustDuration=0.5
    minGustStrength=3
    minTimeBetweenGusts=0
    minUpInterpSpeed=2
    windTurbulenceScalarA=0.1
    windTurbulenceScalarB=1
    windAmplitude=<0 0 0>
    windFrequency=<0 0 0>
    windOffset=<0 0 0>
   }
   riding_Moving_Targeting {
    maxDownInterpSpeed=5
    maxGustDuration=2
    maxGustStrength=7
    maxTimeBetweenGusts=7
    maxUpInterpSpeed=5
    minDownInterpSpeed=2
    minGustDuration=0.5
    minGustStrength=3
    minTimeBetweenGusts=0
    minUpInterpSpeed=2
    windTurbulenceScalarA=0.1
    windTurbulenceScalarB=1
    windAmplitude=<0 0 0>
    windFrequency=<0 0 0>
    windOffset=<0 0 0>
   }
   riding_Standing {
    maxDownInterpSpeed=5
    maxGustDuration=2
    maxGustStrength=7
    maxTimeBetweenGusts=7
    maxUpInterpSpeed=5
    minDownInterpSpeed=2
    minGustDuration=0.5
    minGustStrength=3
    minTimeBetweenGusts=0
    minUpInterpSpeed=2
    windTurbulenceScalarA=0.1
    windTurbulenceScalarB=1
    windAmplitude=<0 0 0>
    windFrequency=<0 0 0>
    windOffset=<0 0 0>
   }
   skydive_DivingDown {
    maxDownInterpSpeed=5
    maxGustDuration=2
    maxGustStrength=7
    maxTimeBetweenGusts=7
    maxUpInterpSpeed=5
    minDownInterpSpeed=2
    minGustDuration=0.5
    minGustStrength=3
    minTimeBetweenGusts=0
    minUpInterpSpeed=2
    windTurbulenceScalarA=0.1
    windTurbulenceScalarB=1
    windAmplitude=<0 0 0>
    windFrequency=<0 0 0>
    windOffset=<0 0 0>
   }
   skydive_FloatingDown {
    maxDownInterpSpeed=5
    maxGustDuration=2
    maxGustStrength=7
    maxTimeBetweenGusts=7
    maxUpInterpSpeed=5
    minDownInterpSpeed=2
    minGustDuration=0.5
    minGustStrength=3
    minTimeBetweenGusts=0
    minUpInterpSpeed=2
    windTurbulenceScalarA=0.1
    windTurbulenceScalarB=1
    windAmplitude=<0 0 0>
    windFrequency=<0 0 0>
    windOffset=<0 0 0>
   }
   skydive_GoingUp {
    maxDownInterpSpeed=5
    maxGustDuration=2
    maxGustStrength=7
    maxTimeBetweenGusts=7
    maxUpInterpSpeed=5
    minDownInterpSpeed=2
    minGustDuration=0.5
    minGustStrength=3
    minTimeBetweenGusts=0
    minUpInterpSpeed=2
    windTurbulenceScalarA=0.1
    windTurbulenceScalarB=1
    windAmplitude=<0 0 0>
    windFrequency=<0 0 0>
    windOffset=<0 0 0>
   }
   skydive_Parachuting {
    maxDownInterpSpeed=5
    maxGustDuration=2
    maxGustStrength=7
    maxTimeBetweenGusts=7
    maxUpInterpSpeed=5
    minDownInterpSpeed=2
    minGustDuration=0.5
    minGustStrength=3
    minTimeBetweenGusts=0
    minUpInterpSpeed=2
    windTurbulenceScalarA=0.1
    windTurbulenceScalarB=1
    windAmplitude=<0 0 0>
    windFrequency=<0 0 0>
    windOffset=<0 0 0>
   }
   skydive_Surfing {
    maxDownInterpSpeed=5
    maxGustDuration=2
    maxGustStrength=7
    maxTimeBetweenGusts=7
    maxUpInterpSpeed=5
    minDownInterpSpeed=2
    minGustDuration=0.5
    minGustStrength=3
    minTimeBetweenGusts=0
    minUpInterpSpeed=2
    windTurbulenceScalarA=0.1
    windTurbulenceScalarB=1
    windAmplitude=<0 0 0>
    windFrequency=<0 0 0>
    windOffset=<0 0 0>
   }
   swimming {
    maxDownInterpSpeed=5
    maxGustDuration=2
    maxGustStrength=7
    maxTimeBetweenGusts=7
    maxUpInterpSpeed=5
    minDownInterpSpeed=2
    minGustDuration=0.5
    minGustStrength=3
    minTimeBetweenGusts=0
    minUpInterpSpeed=2
    windTurbulenceScalarA=0.1
    windTurbulenceScalarB=1
    windAmplitude=<0 0 0>
    windFrequency=<0 0 0>
    windOffset=<0 0 0>
   }
   tactical_Sprint {
    maxDownInterpSpeed=5
    maxGustDuration=2
    maxGustStrength=7
    maxTimeBetweenGusts=7
    maxUpInterpSpeed=5
    minDownInterpSpeed=2
    minGustDuration=0.5
    minGustStrength=3
    minTimeBetweenGusts=0
    minUpInterpSpeed=2
    windTurbulenceScalarA=0.1
    windTurbulenceScalarB=1
    windAmplitude=<0 0 0>
    windFrequency=<0 0 0>
    windOffset=<0 0 0>
   }
   ziplining {
    maxDownInterpSpeed=5
    maxGustDuration=2
    maxGustStrength=7
    maxTimeBetweenGusts=7
    maxUpInterpSpeed=5
    minDownInterpSpeed=2
    minGustDuration=0.5
    minGustStrength=3
    minTimeBetweenGusts=0
    minUpInterpSpeed=2
    windTurbulenceScalarA=0.1
    windTurbulenceScalarB=1
    windAmplitude=<0 0 0>
    windFrequency=<0 0 0>
    windOffset=<0 0 0>
   }
  }
 }
]
